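# Niagara effect for deploying the Believer Female glider, throwing out a short one-shot burst of sticker sprites.
# A single Stickers emitter spawns sprites on a user-supplied skeletal mesh with random SubUV and material variation, using a different upward offset for front end and gameplay.
# Sprites launch outward from a point and upward, then fall under strong gravity and drag while spinning and scaling in and out over under a second.
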NiagaraSystem /BRCosmetics/Gadgets/Assets/VinderTech_GliderChute/Glider_Believer_Female/FX/NS_Glider_Believer_Deploy
dependencies {usedDataInterfaces [&NiagaraDataInterfaceCurve], usedDynamicInputs [], usedModules [&SystemState]}
emitters {Stickers {
  emitterUpdateScript [
   &EmitterState(Life Cycle Mode=::NewEnumerator1, Loop Behavior=::NewEnumerator1, Loop Duration=1, UseLoopDelay=false)
   &SpawnBurst_Instantaneous(Spawn Count=&RandomRangeInt(Minimum=10, Maximum=15), Use Spawn Probability=false)
  ]
  particleSpawnScript [
   &InitializeParticle(
    Lifetime Max=0.8
    Lifetime Min=0.5
    Lifetime Mode=::NewEnumerator1
    Sprite Rotation Mode=::NewEnumerator2
    Sprite Size Mode=::NewEnumerator1
    Uniform Sprite Size Max=30
    Uniform Sprite Size Min=15
    UsePositionOffset=false
    Write Material Random=true
   ) {module=InitializeParticle001}
   &SubUVAnimation(End Frame=16, SubUV Animation Mode:ENiagara_SubUVLookupMode=::NewEnumerator2)
   &SkeletalMeshLocation(
    Apply Sampled Orientation to Mesh Particles=true
    Inherit Velocity=false
    MaskSpawnGroup=false
    Module.Override Randomness=false
    Sampled Position Offset=&FN_FrontEndVector3Modifier(
     Gameplay=&RandomRangeVector(Minimum=<0 0 40>, Maximum=<0 0 50>)
     Front End=&RandomRangeVector(Minimum=<0 0 50>, Maximum=<0 0 70>)
    )
    Skeletal Mesh=$User.SkeletalMesh
   )
   &AddVelocityFromPoint(Module.Use Falloff Distance=false, Velocity Strength=&RandomRangeFloat(Minimum=700, Maximum=1000))
   &AddVelocity(Velocity=&RandomRangeVector(Minimum=<0 0 300>, Maximum=<0 0 600>))
  ]
  particleUpdateScript [
   &GravityForce(Gravity=&RandomRangeVector(Minimum=<0 0 -1500>, Maximum=<0 0 -2500>))
   &Drag(Drag=&RandomRangeFloat(Minimum=2, Maximum=6), Use Linear Drag=true, Use Rotational Drag=false)
   &SolveForcesAndVelocity(Clamp Velocity=false, Limit Acceleration=false)
   &ParticleState()
   &ScaleSpriteSize(Scale Factor=&Vector2DFromFloat(Value=&FloatFromCurve(FloatCurve=DataInterface() {propertyValues {Curve {keys [
          {interpMode=RCIM_Cubic, tangentMode=RCTM_User, time=0, value=0}
          {interpMode=RCIM_Cubic, tangentMode=RCTM_User, time=0.1, value=1}
          {arriveTangent=-0.6933, interpMode=RCIM_Cubic, leaveTangent=-0.6933, tangentMode=RCTM_User, time=0.7, value=0.9}
          {tangentMode=RCTM_User, time=1, value=0}
         ]}}})))
   &SpriteRotationRate(Rotation Rate=&RandomRangeFloat(Minimum=180, Maximum=280))
  ]
  rendererClasses [&NiagaraSpriteRendererProperties]
  renderers [{rendererClass=&NiagaraSpriteRendererProperties}]
 }}
summary {
 systemName=NS_Glider_Believer_Deploy
 emitters [{emitterName=Stickers, rendererClasses [&NiagaraSpriteRendererProperties]}]
 userVariables [
  User.bIsFrontEnd {type=NiagaraBool}
  User.bIsFrontEndPreview {type=NiagaraBool}
  User.SkeletalMesh {
   type=NiagaraDataInterfaceSkeletalMesh
   defaultValue {
    struct=NiagaraExt_VariableValue_DataInterface
    value {dataInterface=&self:NiagaraDataInterfaceSkeletalMesh_0}
   }
  }
 ]
}
