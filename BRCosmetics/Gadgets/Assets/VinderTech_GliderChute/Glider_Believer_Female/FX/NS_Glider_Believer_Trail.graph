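# Niagara trail effect for the Believer (female) glider, active when fully deployed or previewed in the front end.
# Two ribbon AnimTrail emitters each draw two gradient-colored trails from bone pairs.
# GPU Glow, Item and Note sprites spawn in boxes around two skeletal-mesh bones, drifting with limited, inverted owner velocity.
# A system integrator scales gameplay trail alpha and sprite rates, while the front-end flag swaps offsets, velocities and intensity.
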NiagaraSystem /BRCosmetics/Gadgets/Assets/VinderTech_GliderChute/Glider_Believer_Female/FX/NS_Glider_Believer_Trail
dependencies {
 usedDataInterfaces [&NiagaraDataInterfaceColorCurve, &NiagaraDataInterfaceCurve, &NiagaraDataInterfaceArrayFloat3]
 usedDynamicInputs []
 usedModules [&SystemState, &FN_Integrator]
}
emitters {
 FN_AnimTrail {
  emitterSpawnScript [
   &FN_RibbonIDOffsetUtility()
   /SetVariables(Emitter.Use BlackLine=1) {isSetParameters=true}
  ]
  emitterUpdateScript [
   &EmitterState()
   &FN_SpawnAnimTrail(Enable:NiagaraBool=$User.bisFullyDeployed, Number of Anim Trails=2)
  ]
  particleSpawnScript [
   &FN_InitializeAnimTrailBoneIndecies()
   &FN_InitializeAnimTrailBoneIndecies(AnimTrail Selection=1, First Bone or Socket Index=2, Second Bone or Socket Index=3) {module=FN_InitializeAnimTrailBoneIndecies001}
   /SetVariables(Particles.AlphaFadeIn:NiagaraFloat=/Niagara/DynamicInputs/Math/Min_Float(A:NiagaraFloat=&Divide_Float(A=$Emitter.Age, B=1), B:NiagaraFloat=1)) {isSetParameters=true, module=SetVariables_#1}
   &FN_InitializeAnimTrail(
    AnimTrailWidth=0.6
    Color=&MakeLinearColorFromVectorAndFloat("Vector (RGB)"=<1 1 1>, "Float (Alpha)"=$Particles.AlphaFadeIn)
    Lifetime=&Lerp_Float(A=0.2, B=0.25, Alpha=&MakeCustomFloatFromBool(Bool=$User.bisFrontEnd, True Float=1, False Float=0))
    Write Material Random=false
   )
   &AddVelocity(Velocity=&Lerp_Vector(
     A=<0 0 0>
     B=<0 -1200 0>
     Alpha=&MakeCustomFloatFromBool(Bool=$User.bisFrontEnd, True Float=1, False Float=0)
    ))
   &DynamicMaterialParameters(
    Index 0 Param 1=1
    Index 0 Param 3=0.4
    Module.Param0WriteEnabled=true
    Module.Param1WriteEnabled=true
    Module.Param2WriteEnabled=true
    Module.Param3WriteEnabled=true
   )
  ]
  particleUpdateScript [
   &SolveForcesAndVelocity(Clamp Velocity=false, Limit Acceleration=false)
   /Game/Effects/Niagara/Modules/AnimTrail/FN_UpdateAnimTrail/FN_UpdateAnimTrail(New float=1)
   &ScaleColor(
    Scale Alpha=&FloatFromCurve(
     FloatCurve=DataInterface() {propertyValues {Curve {keys [<at 0 0 RCIM_Cubic>, {arriveTangent=2.0248, interpMode=RCIM_Cubic, leaveTangent=2.0248, tangentMode=RCTM_User, time=1, value=1}]}}}
     CurveIndex=&MakeCustomFloatFromBool(
      Bool=$User.bisFrontEnd
      True Float=&FloatFromCurve(FloatCurve=DataInterface() {propertyValues {Curve {keys [<at 0.5 1>,<at 1 0>]}}})
      False Float=$System.FN_Integrator001.Value
     )
     Scale Curve=&FloatFromCurve(FloatCurve=DataInterface() {propertyValues {Curve {keys [<at 0.5 1>,<at 1 0>]}}})
    )
    Scale RGB=&MakeVectorFromLinearColorRGB(Color=&ColorFromCurve(ColorCurve=DataInterface() {propertyValues {
        AlphaCurve {keys [<at 0 1>]}
        BlueCurve {keys [<at 0 0.1302>,<at 0.2158 0.1293>,<at 0.5063 1>,<at 0.7581 1>,<at 1 1>]}
        GreenCurve {keys [<at 0 1>,<at 0.2158 1>,<at 0.5063 0.6664>,<at 0.7581 0.2599>,<at 1 0.1458>]}
        RedCurve {keys [<at 0 0.971>,<at 0.2158 0.1146>,<at 0.5063 0.0469>,<at 0.7581 0.0938>,<at 1 0.8434>]}
       }}))
    ScaleA=true
    ScaleRGB=true
   )
   &ScaleColor(Scale Alpha=&FN_FrontEndFloatModifier(Gameplay=1, Front End=2), ScaleA=true, ScaleRGB=true) {module=ScaleColor001}
   /SetVariables(Particles.RibbonU0Override=$Particles.NormalizedAge) {isSetParameters=true}
   /SetVariables(Particles.OffsetPosition:Vector3f=&Add_Vector(
     A=Linked() {linkedVariable {name=Particles.Position, type=NiagaraPosition}}
     B=&TransformVector(
      Vector=&FN_FrontEndVector3Modifier(
       Gameplay=&SelectVectorFromArray(
        Direct Array Index=$Particles.SpawnGroup
        Direct Array Mode=::NewEnumerator1
        Vector Selection Array=DataInterface() {propertyValues {FloatData [<-25 -10 0>, <25 -10 0>]}}
       )
       Front End=&SelectVectorFromArray(
        Direct Array Index=$Particles.SpawnGroup
        Direct Array Mode=::NewEnumerator1
        Vector Selection Array=DataInterface() {propertyValues {FloatData [<25 0 0>, <-25 0 0>]}}
       )
      )
      Destination Space=::World
      Apply Scale=true
     )
    )) {isSetParameters=true}
  ]
  rendererClasses [&NiagaraRibbonRendererProperties]
  renderers [{rendererClass=&NiagaraRibbonRendererProperties}]
 }
 FN_AnimTrail001 {
  emitterSpawnScript [&FN_RibbonIDOffsetUtility()]
  emitterUpdateScript [
   &EmitterState()
   &FN_SpawnAnimTrail(Enable:NiagaraBool=$User.bisFullyDeployed, Number of Anim Trails=2)
  ]
  particleSpawnScript [
   &FN_InitializeAnimTrailBoneIndecies()
   &FN_InitializeAnimTrailBoneIndecies(AnimTrail Selection=1, First Bone or Socket Index=2, Second Bone or Socket Index=3) {module=FN_InitializeAnimTrailBoneIndecies001}
   /SetVariables(Particles.AlphaFadeIn:NiagaraFloat=&Min_Float(A:NiagaraFloat=&Divide_Float(A=$Emitter.Age, B=1), B:NiagaraFloat=1)) {isSetParameters=true, module=SetVariables_#1}
   &FN_InitializeAnimTrail(
    Color=&MakeLinearColorFromVectorAndFloat("Vector (RGB)"=<1 1 1>, "Float (Alpha)"=&Multiply_Float(A=0.3, B=$Particles.AlphaFadeIn))
    Lifetime=0.25
    Write Material Random=false
   )
   &AddVelocity(Velocity=&Lerp_Vector(
     A=<0 0 0>
     B=<0 -1500 0>
     Alpha=&MakeCustomFloatFromBool(Bool=$User.bisFrontEnd, True Float=1, False Float=0)
    ))
   &DynamicMaterialParameters(
    Index 0 Param 2=1
    Index 0 Param 3=1
    Index 0 Param 4=1
    Module.Param0WriteEnabled=true
    Module.Param1WriteEnabled=true
    Module.Param2WriteEnabled=true
    Module.Param3WriteEnabled=true
   )
  ]
  particleUpdateScript [
   &SolveForcesAndVelocity(Clamp Velocity=false, Limit Acceleration=false)
   &FN_UpdateAnimTrail(New float=1)
   &ScaleColor(
    Scale Alpha=&FloatFromCurve(
     FloatCurve=DataInterface() {propertyValues {Curve {keys [<at 0 0 RCIM_Cubic>, {arriveTangent=2.0248, interpMode=RCIM_Cubic, leaveTangent=2.0248, tangentMode=RCTM_User, time=1, value=1}]}}}
     CurveIndex=&MakeCustomFloatFromBool(Bool=$User.bisFrontEnd, True Float=1, False Float=$System.FN_Integrator001.Value)
    )
    Scale RGB=&MakeVectorFromLinearColorRGB(Color=&ColorFromCurve(ColorCurve=DataInterface() {propertyValues {
        AlphaCurve {keys [<at 0 1>]}
        BlueCurve {keys [<at 0 0.0573>,<at 0.2537 0.2901>,<at 0.5406 0.713>,<at 0.886 1>]}
        GreenCurve {keys [<at 0 1>,<at 0.2537 1>,<at 0.5406 1>,<at 0.886 0.3611>]}
        RedCurve {keys [<at 0 0.9372>,<at 0.2537 0.0938>,<at 0.5406 0.0938>,<at 0.886 0.0417>]}
       }}))
    ScaleA=true
    ScaleRGB=true
   )
   &ScaleColor(Scale Alpha=&FN_FrontEndFloatModifier(Gameplay=1, Front End=2), ScaleA=true, ScaleRGB=true) {module=ScaleColor001}
   /SetVariables(Particles.RibbonU0Override=$Particles.NormalizedAge) {isSetParameters=true}
   /SetVariables(Particles.OffsetPosition:Vector3f=&Add_Vector(
     A=Linked() {linkedVariable {name=Particles.Position, type=NiagaraPosition}}
     B=&TransformVector(
      Vector=&FN_FrontEndVector3Modifier(
       Gameplay=&SelectVectorFromArray(Direct Array Index=$Particles.SpawnGroup, Vector Selection Array=DataInterface() {propertyValues {FloatData [<0 -10 0>, <0 -10 0>]}})
       Front End=&SelectVectorFromArray(Direct Array Index=$Particles.SpawnGroup, Vector Selection Array=DataInterface() {propertyValues {FloatData [<0 10 0>, <0 10 0>]}})
      )
      Destination Space=::World
      Apply Scale=true
     )
    )) {isSetParameters=true}
  ]
  rendererClasses [&NiagaraRibbonRendererProperties]
  renderers [{rendererClass=&NiagaraRibbonRendererProperties}]
 }
 Glow {
  simTarget=GPUComputeSim
  emitterUpdateScript [
   &EmitterState()
   /SetVariables(Emitter.SpawnNum=5) {isSetParameters=true, module=SetVariables_#2}
   &SpawnRate(
    SpawnRate=&MakeCustomFloatFromBool(
     Bool=$User.bisFrontEnd
     True Float=5
     False Float=&MakeCustomFloatFromBool(
      Bool=$User.bisFullyDeployed
      True Float=&Lerp_Float(A=0, B=$Emitter.SpawnNum, Alpha=$System.FN_Integrator001.Value)
      False Float=0
     )
    )
    Use Spawn Probability=false
   )
  ]
  particleSpawnScript [
   &FN_SetSpawnGroup(Number of Spawn Groups=2)
   &InitializeParticle(
    Lifetime Max=0.3
    Lifetime Min=0.2
    Lifetime Mode=::NewEnumerator1
    Sprite Rotation Mode=::NewEnumerator2
    Sprite Size Mode=::NewEnumerator1
    Uniform Sprite Size Max=15
    Uniform Sprite Size Min=10
    UsePositionOffset=false
    Write Material Random=true
   )
   &SkeletalMeshLocation(
    Bone / Socket Index=$Particles.SpawnGroup
    Bone Sampling Mode=::NewEnumerator4
    Inherit Velocity=false
    Interpolate Mesh Sampling=false
    MaskSpawnGroup=false
    Sampled Position Offset=<0 10 0>
    Sampled Position Offset Coordinate Space:ENiagaraMeshTransforms=::NewEnumerator2
    Skeletal Mesh=$User.SkeletalMesh
   )
   &BoxLocation(
    Box Offset=&FN_FrontEndVector3Modifier(
     Gameplay=&SelectVectorFromArray(Direct Array Index=$Particles.SpawnGroup, Vector Selection Array=DataInterface() {propertyValues {FloatData [<40 20 10>, <-40 20 10>]}})
     Front End=&SelectVectorFromArray(Direct Array Index=$Particles.SpawnGroup, Vector Selection Array=DataInterface() {propertyValues {FloatData [<-40 -30 0>, <40 -30 0>]}})
    )
    Box Size=<40 5 5>
    MaskSpawnGroup=false
    Module.Override Seed=false
    Rotate Box=false
    Surface Only=false
   )
   &AddVelocity(Velocity=&FN_FrontEndVector3Modifier(
     Gameplay=&RandomRangeVector(Minimum=<0 200 0>, Maximum=<0 300 0>)
     Front End=&RandomRangeVector(Minimum=<0 -300 0>, Maximum=<0 -400 0>)
    ))
   &CameraOffset(Camera Offset Amount=&RandomRangeFloat(Minimum=0, Maximum=15))
  ]
  particleUpdateScript [
   &ParticleState()
   &Drag(
    Drag=&FN_FrontEndFloatModifier(Gameplay=5, Front End=0)
    Use Linear Drag=true
    Use Rotational Drag=false
   )
   &InheritVelocity(
    Module.Limit Speed=true
    Velocity Limit:NiagaraFloat=&FN_FrontEndFloatModifier(Gameplay=50, Front End=0)
    Velocity Scale=<-1 -1 0>
    Velocity Source=&TransformVector(Vector=$Engine.Owner.Velocity, Source Space=::World, Destination Space=::Local, Apply Scale=true)
   )
   &SolveForcesAndVelocity(Clamp Velocity=false, Limit Acceleration=false)
   &ScaleColor(
    Linear Color Curve=DataInterface() {propertyValues {
      AlphaCurve {keys [<at 0 1>,<at 1 1>]}
      BlueCurve {keys [<at 0 0.1816>,<at 0.1892 0.0576>,<at 0.4707 0.6522>,<at 0.7083 1>,<at 1 1>]}
      GreenCurve {keys [<at 0 0.8306>,<at 0.1892 1>,<at 0.4707 1>,<at 0.7083 0.705>,<at 1 0.1903>]}
      RedCurve {keys [<at 0 1>,<at 0.1892 0.0417>,<at 0.4707 0.026>,<at 0.7083 0.05>,<at 1 0.0833>]}
     }}
    Scale Mode=::NewEnumerator2
   ) {module=ScaleColor001}
   &ScaleColor(
    Scale Alpha=&FloatFromCurve(
     FloatCurve=DataInterface() {propertyValues {Curve {keys [
         <at 0 0 RCIM_Cubic>
         {arriveTangent=-0.0137, interpMode=RCIM_Cubic, leaveTangent=-0.0137, tangentMode=RCTM_User, time=0.2, value=1}
         <at 0.5 0.2 RCIM_Cubic>
         {arriveTangent=-0.0405, interpMode=RCIM_Cubic, leaveTangent=-0.0405, tangentMode=RCTM_User, time=0.8, value=1}
         {arriveTangent=-8.0089, interpMode=RCIM_Cubic, leaveTangent=-8.0089, tangentMode=RCTM_User, time=1, value=0}
        ]}}}
     Scale Curve=&Multiply_Float(A=0.5, B=&FNTimeOfDayScalar(Night Scale=0.5, Morning Scale=0.5, Evening Scale=0.5))
    )
    ScaleA=true
    ScaleRGB=true
   )
   &ScaleSpriteSize(Scale Factor=&Vector2DFromFloat(Value=&FloatFromCurve(FloatCurve=DataInterface() {propertyValues {Curve {keys [<at 0 0>,<at 0.2 1>,<at 0.6 1>,<at 1 0.2>]}}})))
  ]
  rendererClasses [&NiagaraSpriteRendererProperties]
  renderers [{rendererClass=&NiagaraSpriteRendererProperties}]
 }
 Item {
  simTarget=GPUComputeSim
  emitterUpdateScript [
   &EmitterState()
   /SetVariables(Emitter.SpawnNum=20) {isSetParameters=true, module=SetVariables_#2}
   &SpawnRate(
    SpawnRate=&MakeCustomFloatFromBool(
     Bool=$User.bisFrontEnd
     True Float=20
     False Float=&MakeCustomFloatFromBool(
      Bool=$User.bisFullyDeployed
      True Float=&Lerp_Float(A=0, B=$Emitter.SpawnNum, Alpha=$System.FN_Integrator001.Value)
      False Float=0
     )
    )
    Use Spawn Probability=false
   )
  ]
  particleSpawnScript [
   &FN_SetSpawnGroup(Number of Spawn Groups=2)
   &InitializeParticle(
    Lifetime Max=0.6
    Lifetime Min=0.5
    Lifetime Mode=::NewEnumerator1
    Sprite Rotation Mode=::NewEnumerator2
    Sprite Size Mode=::NewEnumerator1
    Uniform Sprite Size Max=25
    Uniform Sprite Size Min=10
    UsePositionOffset=false
    Write Material Random=true
   )
   &SubUVAnimation(End Frame=15, SubUV Animation Mode:ENiagara_SubUVLookupMode=::NewEnumerator2)
   &SkeletalMeshLocation(
    Bone / Socket Index=$Particles.SpawnGroup
    Bone Sampling Mode=::NewEnumerator4
    Inherit Velocity=false
    Interpolate Mesh Sampling=false
    MaskSpawnGroup=false
    Sampled Position Offset=<0 10 0>
    Sampled Position Offset Coordinate Space:ENiagaraMeshTransforms=::NewEnumerator2
    Skeletal Mesh=$User.SkeletalMesh
   )
   &BoxLocation(
    Box Offset=&FN_FrontEndVector3Modifier(
     Gameplay=&SelectVectorFromArray(Direct Array Index=$Particles.SpawnGroup, Vector Selection Array=DataInterface() {propertyValues {FloatData [<40 20 10>, <-40 20 10>]}})
     Front End=&SelectVectorFromArray(Direct Array Index=$Particles.SpawnGroup, Vector Selection Array=DataInterface() {propertyValues {FloatData [<-40 -30 0>, <40 -30 0>]}})
    )
    Box Size=<40 5 5>
    MaskSpawnGroup=false
    Module.Override Seed=false
    Rotate Box=false
    Surface Only=false
   )
   &AddVelocity(Velocity=&FN_FrontEndVector3Modifier(
     Gameplay=&RandomRangeVector(Minimum=<0 200 0>, Maximum=<0 300 0>)
     Front End=&RandomRangeVector(Minimum=<0 -300 0>, Maximum=<0 -400 0>)
    ))
   &CameraOffset(Camera Offset Amount=&RandomRangeFloat(Minimum=0, Maximum=15))
  ]
  particleUpdateScript [
   &ParticleState()
   &Drag(
    Drag=&FN_FrontEndFloatModifier(Gameplay=5, Front End=0)
    Use Linear Drag=true
    Use Rotational Drag=false
   )
   &InheritVelocity(
    Module.Limit Speed=true
    Velocity Limit:NiagaraFloat=&FN_FrontEndFloatModifier(Gameplay=50, Front End=0)
    Velocity Scale=<-1 -1 0>
    Velocity Source=&TransformVector(Vector=$Engine.Owner.Velocity, Source Space=::World, Destination Space=::Local, Apply Scale=true)
   )
   &SolveForcesAndVelocity(Clamp Velocity=false, Limit Acceleration=false)
   &ScaleSpriteSize(Scale Factor=&Vector2DFromFloat(Value=&FloatFromCurve(FloatCurve=DataInterface() {propertyValues {Curve {keys [<at 0 0>,<at 0.2 1>,<at 0.6 1>,<at 1 0.2>]}}})))
   &SpriteRotationRate(Delta Time:NiagaraFloat=$Particles.Lifetime, Rotation Rate=&RandomRangeFloat(Minimum=-3, Maximum=3))
  ]
  rendererClasses [&NiagaraSpriteRendererProperties]
  renderers [{rendererClass=&NiagaraSpriteRendererProperties}]
 }
 Note {
  simTarget=GPUComputeSim
  emitterUpdateScript [
   &EmitterState()
   /SetVariables(Emitter.SpawnNum=10) {isSetParameters=true, module=SetVariables_#2}
   &SpawnRate(
    SpawnRate=&MakeCustomFloatFromBool(
     Bool=$User.bisFrontEnd
     True Float=10
     False Float=&MakeCustomFloatFromBool(
      Bool=$User.bisFullyDeployed
      True Float=&Lerp_Float(A=0, B=$Emitter.SpawnNum, Alpha=$System.FN_Integrator001.Value)
      False Float=0
     )
    )
    Use Spawn Probability=false
   )
  ]
  particleSpawnScript [
   &FN_SetSpawnGroup(Number of Spawn Groups=2)
   &InitializeParticle(
    Color=&LerpLinearColors(
     StartColor=<rgba 0 0 0 0.5>
     EndColor=<rgba 2 2 2 1>
     LerpFactor=&MakeCustomFloatFromBool(
      Bool=&SetBoolByFloatComparison(A=&RandomRangeFloat(Minimum=0, Maximum=1), B=0.5)
      True Float=1
      False Float=0
     )
    )
    Lifetime Max=0.6
    Lifetime Min=0.5
    Lifetime Mode=::NewEnumerator1
    Sprite Rotation Mode=::NewEnumerator2
    Sprite Size Mode=::NewEnumerator1
    Uniform Sprite Size Max=20
    Uniform Sprite Size Min=10
    UsePositionOffset=false
    Write Material Random=true
   )
   &SubUVAnimation(End Frame=3, SubUV Animation Mode:ENiagara_SubUVLookupMode=::NewEnumerator2)
   &SkeletalMeshLocation(
    Bone / Socket Index=$Particles.SpawnGroup
    Bone Sampling Mode=::NewEnumerator4
    Inherit Velocity=false
    Interpolate Mesh Sampling=false
    MaskSpawnGroup=false
    Sampled Position Offset=<0 10 0>
    Sampled Position Offset Coordinate Space:ENiagaraMeshTransforms=::NewEnumerator2
    Skeletal Mesh=$User.SkeletalMesh
   )
   &BoxLocation(
    Box Offset=&FN_FrontEndVector3Modifier(
     Gameplay=&SelectVectorFromArray(Direct Array Index=$Particles.SpawnGroup, Vector Selection Array=DataInterface() {propertyValues {FloatData [<40 20 10>, <-40 20 10>]}})
     Front End=&SelectVectorFromArray(Direct Array Index=$Particles.SpawnGroup, Vector Selection Array=DataInterface() {propertyValues {FloatData [<-40 -30 0>, <40 -30 0>]}})
    )
    Box Size=<40 5 5>
    MaskSpawnGroup=false
    Module.Override Seed=false
    Rotate Box=false
    Surface Only=false
   )
   &AddVelocity(Velocity=&FN_FrontEndVector3Modifier(
     Gameplay=&RandomRangeVector(Minimum=<0 200 0>, Maximum=<0 300 0>)
     Front End=&RandomRangeVector(Minimum=<0 -300 0>, Maximum=<0 -400 0>)
    ))
   &CameraOffset(Camera Offset Amount=&RandomRangeFloat(Minimum=0, Maximum=15))
  ]
  particleUpdateScript [
   &ParticleState()
   &Drag(
    Drag=&FN_FrontEndFloatModifier(Gameplay=5, Front End=0)
    Use Linear Drag=true
    Use Rotational Drag=false
   )
   &InheritVelocity(
    Module.Limit Speed=true
    Velocity Limit:NiagaraFloat=&FN_FrontEndFloatModifier(Gameplay=50, Front End=0)
    Velocity Scale=<-1 -1 0>
    Velocity Source=&TransformVector(Vector=$Engine.Owner.Velocity, Source Space=::World, Destination Space=::Local, Apply Scale=true)
   )
   &SolveForcesAndVelocity(Clamp Velocity=false, Limit Acceleration=false)
   &ScaleColor(
    Linear Color Curve=DataInterface() {propertyValues {
      AlphaCurve {keys [<at 0 1>,<at 1 1>]}
      BlueCurve {keys [<at 0 0.0576>,<at 0.3276 0.6522>,<at 0.6691 1>,<at 1 1>]}
      GreenCurve {keys [<at 0 1>,<at 0.3276 1>,<at 0.6691 0.705>,<at 1 0.1903>]}
      RedCurve {keys [<at 0 0.0417>,<at 0.3276 0.026>,<at 0.6691 0.05>,<at 1 0.0833>]}
     }}
    Scale Mode=::NewEnumerator2
   ) {module=ScaleColor001}
   &ScaleSpriteSize(Scale Factor=&Vector2DFromFloat(Value=&FloatFromCurve(FloatCurve=DataInterface() {propertyValues {Curve {keys [<at 0 0>,<at 0.2 1>,<at 0.6 1>,<at 1 0.2>]}}})))
   &SpriteRotationRate(Delta Time:NiagaraFloat=$Particles.Lifetime, Rotation Rate=&RandomRangeFloat(Minimum=-3, Maximum=3))
  ]
  rendererClasses [&NiagaraSpriteRendererProperties]
  renderers [{rendererClass=&NiagaraSpriteRendererProperties}]
 }
}
summary {
 systemName=NS_Glider_Believer_Trail
 emitters [
  {emitterName=FN_AnimTrail, rendererClasses [&NiagaraRibbonRendererProperties]}
  {emitterName=FN_AnimTrail001, rendererClasses [&NiagaraRibbonRendererProperties]}
  {emitterName=Item, simTarget=GPUComputeSim, rendererClasses [&NiagaraSpriteRendererProperties]}
  {emitterName=Note, simTarget=GPUComputeSim, rendererClasses [&NiagaraSpriteRendererProperties]}
  {emitterName=Glow, simTarget=GPUComputeSim, rendererClasses [&NiagaraSpriteRendererProperties]}
 ]
 userVariables [
  User.bisFrontEnd {type=NiagaraBool}
  User.bIsFrontEndPreview {type=NiagaraBool}
  User.bisFullyDeployed {type=NiagaraBool}
  User.ForwardDot {type=NiagaraFloat, defaultValue=0, description=""}
  User.RibbonIDOffset {type=NiagaraInt32}
  User.RightDot {type=NiagaraFloat, defaultValue=0, description=""}
  User.SkeletalMesh {
   type=NiagaraDataInterfaceSkeletalMesh
   defaultValue {
    struct=NiagaraExt_VariableValue_DataInterface
    value {dataInterface=&self:NiagaraDataInterfaceSkeletalMesh_0}
   }
  }
 ]
}
